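# Physics asset for the collar of the F_MED_FallValleyCharge female player cosmetic, setting up capsule bodies for its dynamic collar bones.
# Seven collar bones (three per side plus a rear one) each constrain to a small root capsule with tight, asymmetric angular limits.
# Only the front-left 01 chain adds a constrained end segment.
# Larger capsules cover the neck and right clavicle and upper arm, joined by free-rotation constraints, with no left-shoulder counterpart.
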PhysicsAsset /BRCosmetics/Characters/Player/Female/Medium/Bodies/F_MED_FallValleyCharge/Meshes/F_MED_FallValleyCharge_Collar_Physics
bodies {
 clavicle_r [{
   shapeType=Capsule
   length=4.5774
   radius=7.9037
   shapeName=clavicle_r_capsule
   center=<-7.3972 0.835 3.442>
   rotation=<rot 1.6884 -90 -74.7975>
  }]
 dyn_collar_01_end_l [{
   shapeType=Capsule
   length=2.7748
   radius=2.6006
   shapeName=dyn_collar_01_end_l_capsule
   center=<-0.105 1.7969 0.9135>
   rotation=<rot 1.1696 5.0672 -104.3322>
  }]
 dyn_collar_01_l [{
   shapeType=Capsule
   length=4.0173
   radius=2.3209
   shapeName=dyn_collar_01_l_capsule
   center=<2.613 -1.3424 -1.4579>
   rotation=<rot -32.2144 12.9399 76.1467>
  }]
 dyn_collar_01_r [{
   shapeType=Capsule
   length=5.0411
   radius=2.3363
   shapeName=dyn_collar_01_r_capsule
   center=<3.952 -0.644 0.8089>
   rotation=<rot -34.0638 11.0303 72.1593>
  }]
 dyn_collar_02_l [{
   shapeType=Capsule
   length=2.4319
   radius=2.5374
   shapeName=dyn_collar_02_l_capsule
   center=<2.3455 1.0287 0.5267>
   rotation=<rot -44.7709 13.3567 63.2576>
  }]
 dyn_collar_02_r [{
   shapeType=Capsule
   length=3.2001
   radius=2.0489
   shapeName=dyn_collar_02_r_capsule
   center=<1.5698 -0.1119 -0.5711>
   rotation=<rot 0 0.0001 84.4531>
  }]
 dyn_collar_03_l [{shapeType=Capsule, length=2.7626, radius=2.5032, shapeName=dyn_collar_03_l_capsule, center=<2.0503 -0.2908 0.216>, rotation=<rot 0 0 90>}]
 dyn_collar_03_r [{shapeType=Capsule, length=2.8898, radius=2.3902, shapeName=dyn_collar_03_r_capsule, center=<2.2541 0.3311 -0.0486>, rotation=<rot 0 0 90>}]
 dyn_collar_rear [{shapeType=Capsule, length=2.0896, radius=2.0067, shapeName=dyn_collar_rear_capsule, center=<2.4511 -0.022 -0.1315>, rotation=<rot 0 0 90>}]
 dyn_collar_root_01_l [{shapeType=Capsule, length=0.505, radius=0.505, shapeName=dyn_collar_root_01_l_capsule, rotation=<rot 0 -90 0>}]
 dyn_collar_root_01_r [{shapeType=Capsule, length=0.505, radius=0.505, shapeName=dyn_collar_root_01_r_capsule, rotation=<rot 0 -90 0>}]
 dyn_collar_root_02_l [{shapeType=Capsule, length=0.505, radius=0.505, shapeName=dyn_collar_root_02_l_capsule, rotation=<rot 0 -90 0>}]
 dyn_collar_root_02_r [{shapeType=Capsule, length=0.505, radius=0.505, shapeName=dyn_collar_root_02_r_capsule, rotation=<rot 0 -90 0>}]
 dyn_collar_root_03_l [{shapeType=Capsule, length=0.505, radius=0.505, shapeName=dyn_collar_root_03_l_capsule, rotation=<rot 0 -90 0>}]
 dyn_collar_root_03_r [{shapeType=Capsule, length=0.505, radius=0.505, shapeName=dyn_collar_root_03_r_capsule, rotation=<rot 0 -90 0>}]
 dyn_collar_root_rear [{shapeType=Capsule, length=0.505, radius=0.505, shapeName=dyn_collar_root_rear_capsule, rotation=<rot 0 -90 0>}]
 neck_01 [{shapeType=Capsule, length=4.5, radius=5.5, shapeName=neck_01_capsule, center=<0 1.0023 0>, rotation=<rot 0 -90.0003 90.0869>}]
 neck_02 [{shapeType=Capsule, length=1.5413, radius=5.125, shapeName=neck_02_capsule, center=<2.1881 0 -0.0007>, rotation=<rot 0 -90.0002 90.0176>}]
 upperarm_twist_01_r [{
   shapeType=Capsule
   length=12.8062
   radius=5.1023
   shapeName=upperarm_twist_01_r_capsule
   center=<-1.1122 0.0309 -0.2359>
   rotation=<rot 0 -104.68 -71.6237>
  }]
}
constraints [
 {
  bone1Name=neck_02
  bone2Name=neck_01
  swing1LimitDegrees=45
  swing1Motion=Free
  swing2LimitDegrees=45
  swing2Motion=Free
  twistLimitDegrees=45
  twistMotion=Free
 }
 {bone1Name=dyn_collar_01_l, bone2Name=dyn_collar_root_01_l, swing1LimitDegrees=5, swing2LimitDegrees=30, twistLimitDegrees=10}
 {bone1Name=dyn_collar_01_r, bone2Name=dyn_collar_root_01_r, swing1LimitDegrees=5, swing2LimitDegrees=30, twistLimitDegrees=30}
 {bone1Name=dyn_collar_02_l, bone2Name=dyn_collar_root_02_l, swing1LimitDegrees=5, swing2LimitDegrees=30, twistLimitDegrees=10}
 {bone1Name=dyn_collar_02_r, bone2Name=dyn_collar_root_02_r, swing1LimitDegrees=5, swing2LimitDegrees=30, twistLimitDegrees=10}
 {bone1Name=dyn_collar_03_l, bone2Name=dyn_collar_root_03_l, swing1LimitDegrees=5, swing2LimitDegrees=30, twistLimitDegrees=10}
 {bone1Name=dyn_collar_03_r, bone2Name=dyn_collar_root_03_r, swing1LimitDegrees=5, swing2LimitDegrees=30, twistLimitDegrees=10}
 {bone1Name=dyn_collar_rear, bone2Name=dyn_collar_root_rear, swing1LimitDegrees=5, swing2LimitDegrees=30, twistLimitDegrees=10}
 {bone1Name=dyn_collar_01_end_l, bone2Name=dyn_collar_01_l, swing1LimitDegrees=12, swing2LimitDegrees=2, twistLimitDegrees=24}
 {
  bone1Name=upperarm_twist_01_r
  bone2Name=clavicle_r
  swing1LimitDegrees=45
  swing1Motion=Free
  swing2LimitDegrees=45
  swing2Motion=Free
  twistLimitDegrees=45
  twistMotion=Free
 }
]
