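# Physics asset for the necklace on the FallValleyCharge medium female body cosmetic, defining its physics bodies and joints.
# Left and right capsule chains run from root bodies, joined to spine_04 with free rotation, through tightly limited links.
# A free-rotation joint bridges the two third links into a loop, and both end box bodies hang off the left third link.
# Large spine_04 and clavicle_r capsules represent the upper body.
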PhysicsAsset /BRCosmetics/Characters/Player/Female/Medium/Bodies/F_MED_FallValleyCharge/Meshes/F_MED_FallValleyCharge_Necklace_Physics
bodies {
 clavicle_r [{
   shapeType=Capsule
   length=9.3213
   radius=7.895
   shapeName=clavicle_r_capsule
   center=<-6.8866 0.1871 3.6901>
   rotation=<rot -19.6382 -85.7461 -88.2474>
  }]
 dyn_necklace_01_l [{
   shapeType=Capsule
   length=2.0988
   radius=0.505
   shapeName=dyn_necklace_01_l_capsule
   center=<0.7724 0.0118 0.0198>
   rotation=<rot 0 89.8341 -89.8431>
  }]
 dyn_necklace_01_l_root [{shapeType=Capsule, length=0.505, radius=0.505, shapeName=dyn_necklace_01_l_root_capsule, rotation=<rot 0 -90 0>}]
 dyn_necklace_01_r [{
   shapeType=Capsule
   length=1.6925
   radius=0.5675
   shapeName=dyn_necklace_01_r_capsule
   center=<0.3742 0.5435 -0.3708>
   rotation=<rot 0 -33.9936 -60.4557>
  }]
 dyn_necklace_01_r_root [{shapeType=Capsule, length=0.505, radius=0.505, shapeName=dyn_necklace_01_r_root_capsule, rotation=<rot 0 -90 0>}]
 dyn_necklace_02_l [{
   shapeType=Capsule
   length=2.4113
   radius=0.505
   shapeName=dyn_necklace_02_l_capsule
   center=<1.2504 0.1666 0.1007>
   rotation=<rot 0 97.0426 -85.4359>
  }]
 dyn_necklace_02_r [{
   shapeType=Capsule
   length=2.5988
   radius=0.505
   shapeName=dyn_necklace_02_r_capsule
   center=<-1.0756 -0.1076 -0.0104>
   rotation=<rot 0 95.832 -89.397>
  }]
 dyn_necklace_03_l [{
   shapeType=Capsule
   length=0.4687
   radius=0.5363
   shapeName=dyn_necklace_03_l_capsule
   center=<-0.4943 0.4874 -0.6927>
   rotation=<rot 23.7268 -90.0001 1.5056>
  }]
 dyn_necklace_03_r [{
   shapeType=Capsule
   length=0.75
   radius=0.5363
   shapeName=dyn_necklace_03_r_capsule
   center=<0.715 0.2553 0.2197>
   rotation=<rot -22.3844 95.3716 -89.5006>
  }]
 dyn_necklace_04_l [{shapeType=Box, extentX=4.5, extentY=1, extentZ=5, shapeName=dyn_necklace_04_l_box, rotation=<rot 0 0 -40.0003>}]
 dyn_necklace_04_r [{
   shapeType=Box
   extentX=5
   extentY=2.0363
   extentZ=5.5
   shapeName=dyn_necklace_04_r_box
   center=<0 2.1094 -2.5139>
   rotation=<rot 0 0 -34.3754>
  }]
 spine_04 [{
   shapeType=Capsule
   length=14.4088
   radius=12.7782
   shapeName=spine_04_capsule
   center=<-0.2898 -1.7685 -0.1019>
   rotation=<rot 0 -118.6591 90>
  }]
}
constraints [
 {bone1Name=dyn_necklace_01_l, bone2Name=dyn_necklace_01_l_root, swing1LimitDegrees=15, swing2LimitDegrees=15, twistLimitDegrees=15}
 {bone1Name=dyn_necklace_02_l, bone2Name=dyn_necklace_01_l, swing1LimitDegrees=15, swing2LimitDegrees=15, twistLimitDegrees=15}
 {bone1Name=dyn_necklace_03_l, bone2Name=dyn_necklace_02_l, swing1LimitDegrees=15, swing2LimitDegrees=15, twistLimitDegrees=15}
 {bone1Name=dyn_necklace_04_l, bone2Name=dyn_necklace_03_l, swing1LimitDegrees=24, swing2LimitDegrees=24, twistLimitDegrees=24}
 {bone1Name=dyn_necklace_01_r, bone2Name=dyn_necklace_01_r_root, swing1LimitDegrees=15, swing2LimitDegrees=15, twistLimitDegrees=15}
 {bone1Name=dyn_necklace_02_r, bone2Name=dyn_necklace_01_r, swing1LimitDegrees=15, swing2LimitDegrees=15, twistLimitDegrees=15}
 {bone1Name=dyn_necklace_03_r, bone2Name=dyn_necklace_02_r, swing1LimitDegrees=15, swing2LimitDegrees=15, twistLimitDegrees=15}
 {
  bone1Name=dyn_necklace_01_r_root
  bone2Name=spine_04
  swing1LimitDegrees=45
  swing1Motion=Free
  swing2LimitDegrees=45
  swing2Motion=Free
  twistLimitDegrees=45
  twistMotion=Free
 }
 {
  bone1Name=dyn_necklace_01_l_root
  bone2Name=spine_04
  swing1LimitDegrees=45
  swing1Motion=Free
  swing2LimitDegrees=45
  swing2Motion=Free
  twistLimitDegrees=45
  twistMotion=Free
 }
 {bone1Name=dyn_necklace_04_r, bone2Name=dyn_necklace_03_l, swing1LimitDegrees=24, swing2LimitDegrees=24, twistLimitDegrees=24}
 {
  bone1Name=dyn_necklace_03_l
  bone2Name=dyn_necklace_03_r
  swing1LimitDegrees=45
  swing1Motion=Free
  swing2LimitDegrees=45
  swing2Motion=Free
  twistLimitDegrees=45
  twistMotion=Free
 }
 {
  bone1Name=clavicle_r
  bone2Name=spine_04
  swing1LimitDegrees=45
  swing1Motion=Free
  swing2LimitDegrees=45
  swing2Motion=Free
  twistLimitDegrees=45
  twistMotion=Free
 }
]
