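# Physics asset for the rear belt of the FallValleyCharge female medium player cosmetic body, giving dynamic motion to two belt sections and a metal tog.
# Each of the three dynamic bones has a capsule body constrained to its own root bone, which carries a tiny capsule.
# Tight angular limits keep the left and right belt pieces nearly rigid, while the metal tog mainly twists with very little swing.
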PhysicsAsset /BRCosmetics/Characters/Player/Female/Medium/Bodies/F_MED_FallValleyCharge/Meshes/F_MED_FallValleyCharge_RearBelt_Physics
bodies {
 dyn_belt_04_metal_tog [{
   shapeType=Capsule
   length=2.375
   radius=1.5
   shapeName=dyn_belt_04_metal_tog_capsule
   center=<-0.1733 0.9693 1.7417>
   rotation=<rot 0 91.4446 -81.0316>
  }]
 dyn_belt_04_metal_tog_root [{shapeType=Capsule, length=0.505, radius=0.505, shapeName=dyn_belt_04_metal_tog_root_capsule, rotation=<rot 0 -90 0>}]
 dyn_belt_05_l [{
   shapeType=Capsule
   length=6.6636
   radius=6.2847
   shapeName=dyn_belt_05_l_capsule
   center=<0.0426 2.3999 0.0872>
   rotation=<rot 0.2472 -94.6884 -88.4673>
  }]
 dyn_belt_05_l_root [{shapeType=Capsule, length=0.505, radius=0.505, shapeName=dyn_belt_05_l_root_capsule, rotation=<rot 0 -90 0>}]
 dyn_belt_06_r [{
   shapeType=Capsule
   length=6.7966
   radius=5.1711
   shapeName=dyn_belt_06_r_capsule
   center=<-0.1241 1.8489 -0.2742>
   rotation=<rot -0.2332 87.0038 1.5913>
  }]
 dyn_belt_06_r_root [{shapeType=Capsule, length=0.505, radius=0.505, shapeName=dyn_belt_06_r_root_capsule, rotation=<rot 0 -90 0>}]
}
constraints [
 {bone1Name=dyn_belt_05_l, bone2Name=dyn_belt_05_l_root, swing1LimitDegrees=4, swing2LimitDegrees=1, twistLimitDegrees=4}
 {bone1Name=dyn_belt_06_r, bone2Name=dyn_belt_06_r_root, swing1LimitDegrees=4, swing2LimitDegrees=1, twistLimitDegrees=4}
 {bone1Name=dyn_belt_04_metal_tog, bone2Name=dyn_belt_04_metal_tog_root, swing1LimitDegrees=1, swing2LimitDegrees=1, twistLimitDegrees=20}
]
