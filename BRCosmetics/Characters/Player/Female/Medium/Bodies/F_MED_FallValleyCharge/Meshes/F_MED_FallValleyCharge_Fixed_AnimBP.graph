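# Post-processing AnimBlueprint for the female FallValleyCharge cosmetic body, adding corrective deformation and secondary motion to a pose copied from the parent mesh.
# A Control Rig applies tunable corrective transforms, then collar offsets feed RigidBody simulations for collar, necklace and rear belt.
# Weighted constraints anchor zipper and belt roots, with AnimDynamics swinging the zippers.
# Thigh test rotations are disconnected from output, and computed right-arm-raise values go unread.
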AnimBlueprint /BRCosmetics/Characters/Player/Female/Medium/Bodies/F_MED_FallValleyCharge/Meshes/F_MED_FallValleyCharge_Fixed_AnimBP
target_skeleton=&/F_MED_FallValleyCharge_Skeleton
graphs [
 AnimGraph {nodes [
   n1 = "Transform (Modify) Bone - Bone: thigh_r"(ComponentPose=Local To Component(), Alpha=1, Translation="0,0,0", Rotation="0.000,0.000,-80.000", Scale="1,1,1") {rotationMode=BMM_Additive, alphaBoolBlend {blendOption=Linear}, boneToModify {boneName=thigh_r}, rotation {roll=-80}}
   n2 = "Transform (Modify) Bone - Bone: thigh_l"(ComponentPose=@n1, Alpha=1, Translation="0,0,0", Rotation="0.000,0.000,-80.000", Scale="1,1,1") {rotationMode=BMM_Additive, alphaBoolBlend {blendOption=Linear}, boneToModify {boneName=thigh_l}, rotation {roll=-80}}
   Component To Local(ComponentPose=@n2)
   n3 = Control Rig!(
    Source=Copy Pose From Mesh() {bUseAttachedParent=true}
    Alpha
    Lower Glute Back:Transform=Get Lower Glute Back() -> (Lower Glute Back:Transform)
    Lower Glute Forward:Transform=Get Lower Glute Forward() -> (Lower Glute Forward:Transform)
    Groin:Transform=Get Groin() -> (Groin:Transform)
    Glute:Transform=Get Deform Value Glute() -> (Deform Value Glute:Transform)
    Glute Forward:Transform=Get Glute Forward() -> (Glute Forward:Transform)
    upperarm_up:Transform=Get Upperarm Up() -> (Upperarm Up:Transform)
    upperarm_fwd_fr:Transform=Get Upperarm Fwd Fr() -> (Upperarm Fwd Fr:Transform)
    upperarm_fwd_bk:Transform=Get Upperarm Fwd Bk() -> (Upperarm Fwd Bk:Transform)
    upperarm_bk:Transform=Get Upperarm Bk() -> (Upperarm Bk:Transform)
    l_thigh_up_belt_l_01:Transform=Get L Thigh Up Belt L 01() -> (L Thigh Up Belt L 01:Transform)
    l_thigh_up_belt_l_02:Transform=Get L Thigh Up Belt L 02() -> (L Thigh Up Belt L 02:Transform)
    l_thigh_up_belt_l_03:Transform=Get L Thigh Up Belt L 03() -> (L Thigh Up Belt L 03:Transform)
    r_thigh_up_belt_r_01:Transform=Get R Thigh Up Belt R 01() -> (R Thigh Up Belt R 01:Transform)
    r_thigh_up_belt_r_02:Transform=Get R Thigh Up Belt R 02() -> (R Thigh Up Belt R 02:Transform)
    r_thigh_up_belt_l_01:Transform=Get R Thigh Up Belt L 01() -> (R Thigh Up Belt L 01:Transform)
    r_thigh_up_belt_r_03:Transform=Get R Thigh Up Belt R 03() -> (R Thigh Up Belt R 03:Transform)
   ) -> (Pose) {
    alphaBoolBlend {blendOption=Linear}
    controlRigAssetReference {blueprintRigClass=&/F_MED_FallValleyCharge_DeformRig_Master_ControlRig.F_MED_FallValleyCharge_DeformRig_Master_ControlRig_C}
   }
   n4 = "Transform (Modify) Bone - Bone: dyn_collar_root_01_r"(ComponentPose=Local To Component(LocalPose=@n3), Alpha=1, Translation="3.227,0.127,-0.191", Rotation="6.057,-35.894,7.193", Scale="1,1,1") {
    rotationMode=BMM_Additive
    translationMode=BMM_Additive
    alphaBoolBlend {blendOption=Linear}
    alphaScaleBiasClamp {bInterpResult=true}
    boneToModify {boneName=dyn_collar_root_01_r}
    rotation=<rot 6.057 -35.894 7.193>
    translation=<3.227 0.127 -0.191>
   }
   n5 = "Transform (Modify) Bone - Bone: dyn_collar_root_01_r"(ComponentPose=@n4, Alpha=1, Translation="-0.866,0.000,0.000", Rotation="-58.861,-12.055,-6.769", Scale="1,1,1") {
    rotationMode=BMM_Additive
    translationMode=BMM_Additive
    alphaBoolBlend {blendOption=Linear}
    alphaScaleBiasClamp {bInterpResult=true}
    boneToModify {boneName=dyn_collar_root_01_r}
    rotation=<rot -58.861 -12.055 -6.769>
    translation {x=-0.866}
   }
   n6 = RigidBody!(ComponentPose=@n5, Alpha, OverrideWorldGravity:Vector="0,0,-980", ExternalForce) -> (Pose) {
    bClampLinearTranslationLimitToRefPose=true
    bOverrideWorldGravity=true
    overridePhysicsAsset=&/F_MED_FallValleyCharge_Collar_Physics
    simulationSpace=BaseBoneSpace
    alphaBoolBlend {blendOption=Linear}
    baseBoneRef {boneName=spine_05}
    componentAppliedLinearAccClamp=<3000 3000 3000>
    componentLinearAccScale=<0.3 0.3 0.3>
    componentLinearVelScale=<0.3 0.3 0.3>
    simSpaceSettings {maxAngularAcceleration=1000, maxAngularVelocity=3, maxLinearAcceleration=250, maxLinearVelocity=500, velocityScaleZ=0, worldAlpha=1}
   }
   n7 = RigidBody(ComponentPose=@n6, OverrideWorldGravity="0,0,-980") {
    bClampLinearTranslationLimitToRefPose=true
    bOverrideWorldGravity=true
    overridePhysicsAsset=&/F_MED_FallValleyCharge_Necklace_Physics
    simulationSpace=BaseBoneSpace
    alphaBoolBlend {blendOption=Linear}
    baseBoneRef {boneName=spine_05}
    componentAppliedLinearAccClamp=<3000 3000 3000>
    componentLinearAccScale=<0.5 0.5 0.5>
    componentLinearVelScale=<0.5 0.5 0.5>
    simSpaceSettings {maxAngularAcceleration=1000, maxAngularVelocity=3, maxLinearAcceleration=250, maxLinearVelocity=500, velocityScaleZ=0, worldAlpha=1}
   }
   n8 = RigidBody(ComponentPose=@n7, OverrideWorldGravity="0,0,-980") {
    bClampLinearTranslationLimitToRefPose=true
    bOverrideWorldGravity=true
    overridePhysicsAsset=&/F_MED_FallValleyCharge_RearBelt_Physics
    simulationSpace=BaseBoneSpace
    alphaBoolBlend {blendOption=Linear}
    baseBoneRef {boneName=pelvis}
    componentAppliedLinearAccClamp=<3000 3000 3000>
    componentLinearAccScale=<0.2 0.2 0.2>
    componentLinearVelScale=<0.2 0.2 0.2>
    simSpaceSettings {maxAngularAcceleration=1000, maxAngularVelocity=3, maxLinearAcceleration=250, maxLinearVelocity=500, velocityScaleZ=0, worldAlpha=1}
   }
   n9 = "Constraint - dyn_zip_arm_l_root (PPPPPPP)"!(
    ComponentPose=@n8
    Alpha=1
    ConstraintWeights_0=0.7519
    ConstraintWeights_1:float=0.1001
    ConstraintWeights_2:float=0.0848
    ConstraintWeights_3:float=0.0299
    ConstraintWeights_4:float=0.0178
    ConstraintWeights_5:float=0.0095
    ConstraintWeights_6:float=0.0061
   ) -> (Pose) {
    alphaBoolBlend {blendOption=Linear}
    boneToModify {boneName=dyn_zip_arm_l_root}
    constraintSetup [
     {targetBone {boneName=lowerarm_twist_02_l}}
     {targetBone {boneName=lowerarm_l}}
     {targetBone {boneName=upperarm_l}}
     {targetBone {boneName=deform_elbow_in_l}}
     {targetBone {boneName=upperarm_twist_02_l}}
     {targetBone {boneName=deform_elbow_out_l}}
     {targetBone {boneName=lowerarm_twist_01_l}}
    ]
    constraintWeights [1,1,1,1,1,1,1]
   }
   n10 = "Constraint - dyn_zip_arm_r_root (PPPPPPP)"!(
    ComponentPose=@n9
    Alpha=1
    ConstraintWeights_0=0.8063
    ConstraintWeights_1:float=0.1096
    ConstraintWeights_2:float=0.0361
    ConstraintWeights_3:float=0.0333
    ConstraintWeights_4:float=0.0083
    ConstraintWeights_5:float=0.0055
    ConstraintWeights_6:float=0.0008
   ) -> (Pose) {
    alphaBoolBlend {blendOption=Linear}
    boneToModify {boneName=dyn_zip_arm_r_root}
    constraintSetup [
     {targetBone {boneName=lowerarm_twist_02_r}}
     {targetBone {boneName=lowerarm_r}}
     {targetBone {boneName=upperarm_r}}
     {targetBone {boneName=lowerarm_twist_01_r}}
     {targetBone {boneName=deform_elbow_in_r}}
     {targetBone {boneName=deform_elbow_out_r}}
     {targetBone {boneName=upperarm_twist_02_r}}
    ]
    constraintWeights [1,1,1,1,1,1,1]
   }
   n11 = "Constraint - dyn_zip_chest_root (PPPPPPPPP)"!(
    ComponentPose=@n10
    Alpha=1
    ConstraintWeights_0=0.3776
    ConstraintWeights_1:float=0.3338
    ConstraintWeights_2:float=0.1457
    ConstraintWeights_3:float=0.11
    ConstraintWeights_4:float=0.0329
    ConstraintWeights_5:float=0
    ConstraintWeights_6:float=0
    ConstraintWeights_7:float=0
    ConstraintWeights_8:float=0
   ) -> (Pose) {
    alphaBoolBlend {blendOption=Linear}
    boneToModify {boneName=dyn_zip_chest_root}
    constraintSetup [
     {targetBone {boneName=spine_02}}
     {targetBone {boneName=spine_04}}
     {targetBone {boneName=pelvis}}
     {targetBone {boneName=spine_03}}
     {targetBone {boneName=spine_01}}
     {targetBone {boneName=deform_groin_r}}
     {targetBone {boneName=deform_groin_l}}
     {targetBone {boneName=thigh_twist_01_r}}
     {targetBone {boneName=thigh_twist_01_l}}
    ]
    constraintWeights [1,1,1,1,1,1,1,1,1]
   }
   n12 = "Constraint - dyn_zip_01_l_root (PPPPP)"!(
    ComponentPose=@n11
    Alpha=1
    ConstraintWeights_0=0.4948
    ConstraintWeights_1:float=0.2901
    ConstraintWeights_2:float=0.1282
    ConstraintWeights_3:float=0.0453
    ConstraintWeights_4:float=0.0417
   ) -> (Pose) {
    alphaBoolBlend {blendOption=Linear}
    boneToModify {boneName=dyn_zip_01_l_root}
    constraintSetup [
     {targetBone {boneName=spine_04}}
     {targetBone {boneName=spine_02}}
     {targetBone {boneName=spine_03}}
     {targetBone {boneName=pelvis}}
     {targetBone {boneName=spine_01}}
    ]
    constraintWeights [1,1,1,1,1]
   }
   n13 = "Constraint - dyn_zip_02_l_root (PPPPPPPP)"!(
    ComponentPose=@n12
    Alpha=1
    ConstraintWeights_0=0.3794
    ConstraintWeights_1:float=0.3772
    ConstraintWeights_2:float=0.1928
    ConstraintWeights_3:float=0.028
    ConstraintWeights_4:float=0.0133
    ConstraintWeights_5:float=0.0081
    ConstraintWeights_6:float=0.0009
    ConstraintWeights_7:float=0.0002
   ) -> (Pose) {
    alphaBoolBlend {blendOption=Linear}
    boneToModify {boneName=dyn_zip_02_l_root}
    constraintSetup [
     {targetBone {boneName=deform_pec_l}}
     {targetBone {boneName=spine_04}}
     {targetBone {boneName=spine_05}}
     {targetBone {boneName=clavicle_l}}
     {targetBone {boneName=deform_upperarm_fr_l}}
     {targetBone {boneName=upperarm_twist_01_l}}
     {targetBone {boneName=spine_02}}
     {targetBone {boneName=spine_03}}
    ]
    constraintWeights [1,1,1,1,1,1,1,1]
   }
   n14 = "Constraint - dyn_zip_01_r_root (PPPPP)"!(
    ComponentPose=@n13
    Alpha=1
    ConstraintWeights_0=0.5217
    ConstraintWeights_1:float=0.2716
    ConstraintWeights_2:float=0.1263
    ConstraintWeights_3:float=0.0407
    ConstraintWeights_4:float=0.0398
   ) -> (Pose) {
    alphaBoolBlend {blendOption=Linear}
    boneToModify {boneName=dyn_zip_01_r_root}
    constraintSetup [
     {targetBone {boneName=spine_04}}
     {targetBone {boneName=spine_02}}
     {targetBone {boneName=spine_03}}
     {targetBone {boneName=pelvis}}
     {targetBone {boneName=spine_01}}
    ]
    constraintWeights [1,1,1,1,1]
   }
   n15 = Anim Dynamics - Bone: dyn_zip_01_r(ComponentPose=@n14, Alpha=1, ExternalForce="0,0,0") {
    angularDampingOverride=0.9
    bOverrideAngularDamping=true
    bUseGravityOverride=true
    simulationSpace=RootRelative
    alphaBoolBlend {blendOption=Linear}
    boundBone {boneName=dyn_zip_01_r}
    gravityOverride {z=-980}
    physicsBodyDefinitions [{
      boundBone {boneName=dyn_zip_01_r}
      boxExtents=<10 5 5>
      constraintSetup {angularTargetAxis=AxisY, angularLimitsMax=<5 5 10>, angularLimitsMin=<-5 -5 -10>, angularTarget=<0 0 0>}
      localJointOffset=<-2.5 0 0>
     }]
    relativeSpaceBone {boneName=dyn_zip_01_r_root}
   }
   n16 = Anim Dynamics - Bone: dyn_zip_01_l(ComponentPose=@n15, Alpha=1, ExternalForce="0,0,0") {
    angularDampingOverride=0.9
    bOverrideAngularDamping=true
    bUseGravityOverride=true
    simulationSpace=RootRelative
    alphaBoolBlend {blendOption=Linear}
    boundBone {boneName=dyn_zip_01_l}
    gravityOverride {z=-980}
    physicsBodyDefinitions [{
      boundBone {boneName=dyn_zip_01_l}
      boxExtents=<10 5 5>
      constraintSetup {angularTargetAxis=AxisY, angularLimitsMax=<5 5 10>, angularLimitsMin=<-5 -5 -10>, angularTarget=<0 0 0>}
      localJointOffset=<-2.5 0 0>
     }]
    relativeSpaceBone {boneName=dyn_zip_01_r_root}
   }
   n17 = Anim Dynamics - Bone: dyn_zip_02_l(ComponentPose=@n16, Alpha=1, ExternalForce="0,0,0") {
    angularDampingOverride=0.9
    bOverrideAngularDamping=true
    bUseGravityOverride=true
    simulationSpace=RootRelative
    alphaBoolBlend {blendOption=Linear}
    boundBone {boneName=dyn_zip_02_l}
    gravityOverride {z=-980}
    physicsBodyDefinitions [{
      boundBone {boneName=dyn_zip_02_l}
      boxExtents=<10 5 5>
      constraintSetup {angularTargetAxis=AxisY, angularLimitsMax=<5 5 10>, angularLimitsMin=<-5 -5 -10>, angularTarget=<0 0 0>}
      localJointOffset=<-1 0 0>
     }]
    relativeSpaceBone {boneName=dyn_zip_01_r_root}
   }
   n18 = Anim Dynamics - Bone: dyn_zip_chest(ComponentPose=@n17, Alpha=1, ExternalForce="0,0,0") {
    angularDampingOverride=0.9
    bOverrideAngularDamping=true
    bUseGravityOverride=true
    simulationSpace=RootRelative
    alphaBoolBlend {blendOption=Linear}
    boundBone {boneName=dyn_zip_chest}
    gravityOverride {z=-980}
    physicsBodyDefinitions [{
      boundBone {boneName=dyn_zip_chest}
      boxExtents=<10 5 5>
      constraintSetup {angularTargetAxis=AxisY, angularLimitsMax=<5 5 10>, angularLimitsMin=<-5 -5 -10>, angularTarget=<0 0 0>}
      localJointOffset=<-2.5 0 0>
     }]
    relativeSpaceBone {boneName=dyn_zip_01_r_root}
   }
   n19 = Anim Dynamics - Bone: dyn_zip_arm_r(ComponentPose=@n18, Alpha=1, ExternalForce="0,0,0") {
    angularDampingOverride=0.9
    bOverrideAngularDamping=true
    bUseGravityOverride=true
    simulationSpace=RootRelative
    alphaBoolBlend {blendOption=Linear}
    boundBone {boneName=dyn_zip_arm_r}
    gravityOverride {z=-980}
    physicsBodyDefinitions [{
      boundBone {boneName=dyn_zip_arm_r}
      boxExtents=<10 5 5>
      constraintSetup {angularTargetAxis=AxisY, angularLimitsMax=<5 5 90>, angularLimitsMin=<-5 -5 -10>, angularTarget=<0 0 0>}
      localJointOffset=<2.5 0 0>
     }]
    relativeSpaceBone {boneName=dyn_zip_01_r_root}
   }
   n20 = Anim Dynamics - Bone: dyn_zip_arm_l(ComponentPose=@n19, Alpha=1, ExternalForce="0,0,0") {
    angularDampingOverride=0.9
    bOverrideAngularDamping=true
    bUseGravityOverride=true
    simulationSpace=RootRelative
    alphaBoolBlend {blendOption=Linear}
    boundBone {boneName=dyn_zip_arm_l}
    gravityOverride {z=-980}
    physicsBodyDefinitions [{
      boundBone {boneName=dyn_zip_arm_l}
      boxExtents=<10 5 5>
      constraintSetup {angularTargetAxis=AxisY, angularLimitsMax=<5 5 90>, angularLimitsMin=<-5 -5 -10>, angularTarget=<0 0 0>}
      localJointOffset=<-2.5 0 0>
     }]
    relativeSpaceBone {boneName=dyn_zip_01_r_root}
   }
   n21 = "Constraint - dyn_belt_01_l_root (PPPPPPPP)"!(
    ComponentPose=@n20
    Alpha=1
    ConstraintWeights_0=0.5399
    ConstraintWeights_1:float=0.2246
    ConstraintWeights_2:float=0.1968
    ConstraintWeights_3:float=0.0269
    ConstraintWeights_4:float=0.006
    ConstraintWeights_5:float=0.0045
    ConstraintWeights_6:float=0
    ConstraintWeights_7:float=0
   ) -> (Pose) {
    alphaBoolBlend {blendOption=Linear}
    boneToModify {boneName=dyn_belt_01_l_root}
    constraintSetup [
     {targetBone {boneName=pelvis}}
     {targetBone {boneName=thigh_twist_01_l}}
     {targetBone {boneName=deform_groin_l}}
     {targetBone {boneName=thigh_twist_01_r}}
     {targetBone {boneName=thigh_l}}
     {targetBone {boneName=deform_groin_r}}
     {targetBone {boneName=spine_02}}
     {targetBone {boneName=spine_04}}
    ]
    constraintWeights [1,1,1,1,1,1,1,1]
   }
   n22 = "Constraint - dyn_belt_02_l_root (PPPPPPPP)"!(
    ComponentPose=@n21
    Alpha=1
    ConstraintWeights_0=0.3806
    ConstraintWeights_1:float=0.3058
    ConstraintWeights_2:float=0.2788
    ConstraintWeights_3:float=0.0242
    ConstraintWeights_4:float=0.0087
    ConstraintWeights_5:float=0.0011
    ConstraintWeights_6:float=0.0007
    ConstraintWeights_7:float=0
   ) -> (Pose) {
    alphaBoolBlend {blendOption=Linear}
    boneToModify {boneName=dyn_belt_02_l_root}
    constraintSetup [
     {targetBone {boneName=pelvis}}
     {targetBone {boneName=thigh_twist_01_l}}
     {targetBone {boneName=deform_groin_l}}
     {targetBone {boneName=thigh_l}}
     {targetBone {boneName=thigh_twist_01_r}}
     {targetBone {boneName=deform_groin_r}}
     {targetBone {boneName=spine_02}}
     {targetBone {boneName=deform_glute_l}}
    ]
    constraintWeights [1,1,1,1,1,1,1,1]
   }
   n23 = "Constraint - dyn_belt_03_l_root (PPPPPP)"!(
    ComponentPose=@n22
    Alpha=1
    ConstraintWeights_0=0.45
    ConstraintWeights_1:float=0.55
    ConstraintWeights_2:float=0
    ConstraintWeights_3:float=0
    ConstraintWeights_4:float=0
    ConstraintWeights_5:float=0
   ) -> (Pose) {
    alphaBoolBlend {blendOption=Linear}
    boneToModify {boneName=dyn_belt_03_l_root}
    constraintSetup [
     {targetBone {boneName=thigh_twist_01_l}}
     {targetBone {boneName=pelvis}}
     {targetBone {boneName=deform_groin_l}}
     {targetBone {boneName=deform_glute_l}}
     {targetBone {boneName=thigh_l}}
     {targetBone {boneName=spine_05}}
    ]
    constraintWeights [1,1,1,1,1,1]
   }
   n24 = "Constraint - dyn_belt_04_l_root (PPPPP)"!(
    ComponentPose=@n23
    Alpha=1
    ConstraintWeights_0=0.5684
    ConstraintWeights_1:float=0.3342
    ConstraintWeights_2:float=0.0974
    ConstraintWeights_3:float=0
    ConstraintWeights_4:float=0
   ) -> (Pose) {
    alphaBoolBlend {blendOption=Linear}
    boneToModify {boneName=dyn_belt_04_l_root}
    constraintSetup [
     {targetBone {boneName=pelvis}}
     {targetBone {boneName=deform_glute_l}}
     {targetBone {boneName=thigh_twist_01_l}}
     {targetBone {boneName=deform_glute_r}}
     {targetBone {boneName=thigh_l}}
    ]
    constraintWeights [1,1,1,1,1]
   }
   n25 = "Constraint - dyn_belt_01_r_root (PPPPPPPPP)"!(
    ComponentPose=@n24
    Alpha=1
    ConstraintWeights_0=0.6509
    ConstraintWeights_1:float=0.1394
    ConstraintWeights_2:float=0.0949
    ConstraintWeights_3:float=0.0763
    ConstraintWeights_4:float=0.0337
    ConstraintWeights_5:float=0.0031
    ConstraintWeights_6:float=0.0015
    ConstraintWeights_7:float=0
    ConstraintWeights_8:float=0
   ) -> (Pose) {
    alphaBoolBlend {blendOption=Linear}
    boneToModify {boneName=dyn_belt_01_r_root}
    constraintSetup [
     {targetBone {boneName=pelvis}}
     {targetBone {boneName=thigh_twist_01_r}}
     {targetBone {boneName=deform_groin_r}}
     {targetBone {boneName=thigh_twist_01_l}}
     {targetBone {boneName=deform_groin_l}}
     {targetBone {boneName=spine_02}}
     {targetBone {boneName=thigh_r}}
     {targetBone {boneName=thigh_l}}
     {targetBone {boneName=spine_04}}
    ]
    constraintWeights [1,1,1,1,1,1,1,1,1]
   }
   n26 = "Constraint - dyn_belt_02_r_root (PPPPPPP)"!(
    ComponentPose=@n25
    Alpha=1
    ConstraintWeights_0=0.3689
    ConstraintWeights_1:float=0.3278
    ConstraintWeights_2:float=0.2685
    ConstraintWeights_3:float=0.0261
    ConstraintWeights_4:float=0.0076
    ConstraintWeights_5:float=0.0008
    ConstraintWeights_6:float=0.0003
   ) -> (Pose) {
    alphaBoolBlend {blendOption=Linear}
    boneToModify {boneName=dyn_belt_02_r_root}
    constraintSetup [
     {targetBone {boneName=pelvis}}
     {targetBone {boneName=thigh_twist_01_r}}
     {targetBone {boneName=deform_groin_r}}
     {targetBone {boneName=thigh_r}}
     {targetBone {boneName=thigh_twist_01_l}}
     {targetBone {boneName=deform_groin_l}}
     {targetBone {boneName=spine_02}}
    ]
    constraintWeights [1,1,1,1,1,1,1]
   }
   n27 = "Constraint - dyn_belt_03_r_root (PPPPP)"!(
    ComponentPose=@n26
    Alpha=1
    ConstraintWeights_0=0.5574
    ConstraintWeights_1:float=0.2241
    ConstraintWeights_2:float=0.1275
    ConstraintWeights_3:float=0.0852
    ConstraintWeights_4:float=0.0057
   ) -> (Pose) {
    alphaBoolBlend {blendOption=Linear}
    boneToModify {boneName=dyn_belt_03_r_root}
    constraintSetup [
     {targetBone {boneName=thigh_twist_01_r}}
     {targetBone {boneName=pelvis}}
     {targetBone {boneName=deform_groin_r}}
     {targetBone {boneName=thigh_r}}
     {targetBone {boneName=deform_glute_r}}
    ]
    constraintWeights [1,1,1,1,1]
   }
   n28 = "Constraint - dyn_belt_04_r_root (PPPPPP)"!(
    ComponentPose=@n27
    Alpha=1
    ConstraintWeights_0=0.185
    ConstraintWeights_1:float=0.8
    ConstraintWeights_2:float=0.13
    ConstraintWeights_3:float=0.02
    ConstraintWeights_4:float=0
    ConstraintWeights_5:float=0
   ) -> (Pose) {
    alphaBoolBlend {blendOption=Linear}
    boneToModify {boneName=dyn_belt_04_r_root}
    constraintSetup [
     {targetBone {boneName=thigh_twist_01_r}}
     {targetBone {boneName=pelvis}}
     {targetBone {boneName=deform_glute_r}}
     {targetBone {boneName=thigh_r}}
     {targetBone {boneName=deform_groin_r}}
     {targetBone {boneName=deform_glute_lower_r}}
    ]
    constraintWeights [1,1,1,1,1,1]
   }
   n29 = "Constraint - dyn_belt_05_l_root (PPPPPPPP)"!(
    ComponentPose=@n28
    Alpha=1
    ConstraintWeights_0=0.6417
    ConstraintWeights_1:float=0.2011
    ConstraintWeights_2:float=0.0641
    ConstraintWeights_3:float=0.0463
    ConstraintWeights_4:float=0.0374
    ConstraintWeights_5:float=0.0088
    ConstraintWeights_6:float=0.0005
    ConstraintWeights_7:float=0.0001
   ) -> (Pose) {
    alphaBoolBlend {blendOption=Linear}
    boneToModify {boneName=dyn_belt_05_l_root}
    constraintSetup [
     {targetBone {boneName=pelvis}}
     {targetBone {boneName=deform_glute_l}}
     {targetBone {boneName=deform_groin_l}}
     {targetBone {boneName=deform_glute_r}}
     {targetBone {boneName=thigh_twist_01_l}}
     {targetBone {boneName=thigh_twist_01_r}}
     {targetBone {boneName=thigh_l}}
     {targetBone {boneName=thigh_r}}
    ]
    constraintWeights [1,1,1,1,1,1,1,1]
   }
   n30 = "Constraint - dyn_belt_06_r_root (PPPPPPPPP)"!(
    ComponentPose=@n29
    Alpha=1
    ConstraintWeights_0=0.632
    ConstraintWeights_1:float=0.2181
    ConstraintWeights_2:float=0.0703
    ConstraintWeights_3:float=0.0368
    ConstraintWeights_4:float=0.0358
    ConstraintWeights_5:float=0.0065
    ConstraintWeights_6:float=0.0003
    ConstraintWeights_7:float=0.0001
    ConstraintWeights_8:float=0
   ) -> (Pose) {
    alphaBoolBlend {blendOption=Linear}
    boneToModify {boneName=dyn_belt_06_r_root}
    constraintSetup [
     {targetBone {boneName=pelvis}}
     {targetBone {boneName=deform_glute_r}}
     {targetBone {boneName=deform_groin_r}}
     {targetBone {boneName=thigh_twist_01_r}}
     {targetBone {boneName=deform_glute_l}}
     {targetBone {boneName=thigh_twist_01_l}}
     {targetBone {boneName=thigh_r}}
     {targetBone {boneName=thigh_l}}
     {targetBone {boneName=spine_02}}
    ]
    constraintWeights [1,1,1,1,1,1,1,1,1]
   }
   n31 = "Constraint - dyn_belt_04_metal_tog_root (PPPP)"!(
    ComponentPose=@n30
    Alpha=1
    ConstraintWeights_0=0.04
    ConstraintWeights_1:float=0.93
    ConstraintWeights_2:float=0.02
    ConstraintWeights_3:float=0.01
   ) -> (Pose) {
    alphaBoolBlend {blendOption=Linear}
    boneToModify {boneName=dyn_belt_04_metal_tog_root}
    constraintSetup [
     {targetBone {boneName=dyn_belt_06_r}}
     {targetBone {boneName=dyn_belt_04_r}}
     {targetBone {boneName=dyn_belt_03_r}}
     {targetBone {boneName=dyn_belt_05_l}}
    ]
    constraintWeights [1,1,1,1]
   }
   Output Pose(Result=Component To Local(ComponentPose=@n31)) {layerGroup=None}
  ]}
 EventGraph {nodes [
   Event BlueprintUpdateAnimation()
   n1 = BreakTransform!(InTransform=GetDeltaTransformFromRefPose(self=GetOwnerSkeletalMeshComponent(), BoneName=upperarm_r, BaseName=spine_05)) -> (Location, Rotation_Roll:float, Rotation_Pitch:float, Rotation_Yaw:float, Scale)
   > Set R Arm Up 90(R Arm Up 90:double=MapRangeClamped(Value=@n1.Rotation_Yaw, InRangeA=65, InRangeB=90)) -> (Output_Get:double)
   > Set R Arm Up 90 to 130(R Arm Up 90 to 130:double=MapRangeClamped(Value=@n1.Rotation_Yaw, InRangeA=90, InRangeB=130)) -> (Output_Get:double)
   TryGetPawnOwner()
  ]}
]
variables [
 Deform Value Glute {type=Transform, default {location=<-5 -2.5 0>, rotation=<rot 0 0 0>, scale=<1 1.3 1>}}
 Glute Forward {type=Transform, default {location=<-7.5 -3.3 0>, rotation=<rot 0 0 0>, scale=<1 1.3 1>}}
 Groin {type=Transform, default {location=<-6 22 0>, rotation=<rot 0 0 0>, scale=<0.5 1 1>}}
 L Thigh Up Belt L 01 {type=Transform, default {location=<-1 2 2>, rotation=<rot 30 0 0>, scale=<1 1 1>}}
 L Thigh Up Belt L 02 {type=Transform, default {location=<1 1 -1>, rotation=<rot 0 0 0>, scale=<1 1 1>}}
 L Thigh Up Belt L 03 {type=Transform, default {location=<2.5 0.8 0>, rotation=<rot 12 -10 0>, scale=<1 1 1>}}
 Lower Glute Back {type=Transform, default {location=<0 0 0>, rotation=<rot 0 0 0>, scale=<1 1 1>}}
 Lower Glute Forward {type=Transform, default {location=<-6 -40 0>, rotation=<rot 0 0 0>, scale=<1 1 1>}}
 R Arm Up 90 {type=number}
 R Arm Up 90 to 130 {type=number}
 R Thigh Up Belt L 01 {type=Transform, default {location=<2 0 -1>, rotation=<rot -20 0 0>, scale=<1 1 1>}}
 R Thigh Up Belt R 01 {type=Transform, default {location=<3 2 0>, rotation=<rot -10 0 0>, scale=<1 1 1>}}
 R Thigh Up Belt R 02 {type=Transform, default {location=<0.5 1.5 -2>, rotation=<rot -20 -20 0>, scale=<1 1 1>}}
 R Thigh Up Belt R 03 {type=Transform, default {location=<1 0 0>, rotation=<rot 0 0 0>, scale=<1 1 1>}}
 Upperarm Bk {type=Transform, default {location=<0 -4 0>, rotation=<rot 0 0 0>, scale=<1 1 1>}}
 Upperarm Fwd Bk {type=Transform, default {location=<0 -6 0>, rotation=<rot 0 0 0>, scale=<1 1 1>}}
 Upperarm Fwd Fr {type=Transform, default {location=<-3 15 -2>, rotation=<rot 0 0 0>, scale=<2 1 1>}}
 Upperarm Up {type=Transform, default {location=<0 0 8>, rotation=<rot 0 0 0>, scale=<1 1 1>}}
]
