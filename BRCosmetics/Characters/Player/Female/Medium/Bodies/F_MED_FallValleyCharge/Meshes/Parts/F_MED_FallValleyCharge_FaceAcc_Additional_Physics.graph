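# Physics asset for hair in the FaceAcc_Additional part of the female medium FallValleyCharge cosmetic, with capsule bodies on head and dynamic hair bones.
# Left and right front strands chain from head through root bones, with two short side branches on the right.
# The ponytail hangs from head at bone 04, which branches into three sub-chains.
# Narrow angular limits restrain motion, with the main ponytail allowing wider swing than twist.
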PhysicsAsset /BRCosmetics/Characters/Player/Female/Medium/Bodies/F_MED_FallValleyCharge/Meshes/Parts/F_MED_FallValleyCharge_FaceAcc_Additional_Physics
bodies {
 dyn_hair_front_01_l [{shapeType=Capsule, length=1.5, radius=1, shapeName=dyn_hair_front_01_l_capsule, center=<1 0 0>, rotation=<rot 90 0 0>}]
 dyn_hair_front_01_l_root [{shapeType=Capsule, length=0.505, radius=0.505, shapeName=dyn_hair_front_01_l_root_capsule, rotation=<rot 0 -90 0>}]
 dyn_hair_front_01_r [{shapeType=Capsule, length=2, radius=1.5, shapeName=dyn_hair_front_01_r_capsule, center=<1 0 0>, rotation=<rot 90 0 0>}]
 dyn_hair_front_01_r_root [{shapeType=Capsule, length=0.505, radius=0.505, shapeName=dyn_hair_front_01_r_root_capsule, rotation=<rot 0 -90 0>}]
 dyn_hair_front_02_l [{shapeType=Capsule, length=1.5, radius=1, shapeName=dyn_hair_front_02_l_capsule, center=<1 0 0>, rotation=<rot 90 0 0>}]
 dyn_hair_front_02_r [{shapeType=Capsule, length=2, radius=1.5, shapeName=dyn_hair_front_02_r_capsule, center=<1 0 0>, rotation=<rot 90 0 0>}]
 dyn_hair_front_03_l [{shapeType=Capsule, length=1.5, radius=1, shapeName=dyn_hair_front_03_l_capsule, center=<1 0 0>, rotation=<rot 90 0 0>}]
 dyn_hair_front_03_r [{shapeType=Capsule, length=2, radius=1.5, shapeName=dyn_hair_front_03_r_capsule, center=<1 0 0>, rotation=<rot 90 0 0>}]
 dyn_hair_front_04_l [{shapeType=Capsule, length=1.5, radius=1, shapeName=dyn_hair_front_04_l_capsule, center=<1 0 0>, rotation=<rot 90 0 0>}]
 dyn_hair_front_04_r [{shapeType=Capsule, length=2, radius=1.5, shapeName=dyn_hair_front_04_r_capsule, center=<1 0 0>, rotation=<rot 90 0 0>}]
 dyn_hair_front_05_r [{
   shapeType=Capsule
   length=0.505
   radius=0.505
   shapeName=dyn_hair_front_05_r_capsule
   center=<0.8934 0.1392 0.0564>
   rotation=<rot 0 60.3014 -83.8652>
  }]
 dyn_hair_front_06_r [{
   shapeType=Capsule
   length=0.505
   radius=0.505
   shapeName=dyn_hair_front_06_r_capsule
   center=<1.0106 0.0971 0.1426>
   rotation=<rot 0 97.3116 -80.176>
  }]
 dyn_hair_front_07_r [{
   shapeType=Capsule
   length=0.5521
   radius=0.4403
   shapeName=dyn_hair_front_07_r_capsule
   center=<0.8491 0.0486 -0.1154>
   rotation=<rot 0 53.7257 -82.3924>
  }]
 dyn_hair_front_08_r [{
   shapeType=Capsule
   length=0.505
   radius=0.505
   shapeName=dyn_hair_front_08_r_capsule
   center=<0.9013 0.1949 0.1229>
   rotation=<rot 0 105.384 -79.442>
  }]
 dyn_hair_ponytail_04 [{shapeType=Capsule, length=5.5559, radius=4.6152, shapeName=dyn_hair_ponytail_04_capsule, center=<1 0 0>, rotation=<rot 90 0 0>}]
 dyn_hair_ponytail_05 [{shapeType=Capsule, length=4.1156, radius=4.4873, shapeName=dyn_hair_ponytail_05_capsule, center=<1 0 0>, rotation=<rot 90 0 0>}]
 dyn_hair_ponytail_12 [{
   shapeType=Capsule
   length=1.3115
   radius=0.6721
   shapeName=dyn_hair_ponytail_12_capsule
   center=<1.372 0.0622 -0.1476>
   rotation=<rot 0 93.7703 -72.6956>
  }]
 dyn_hair_ponytail_13 [{
   shapeType=Capsule
   length=1.0938
   radius=0.3246
   shapeName=dyn_hair_ponytail_13_capsule
   center=<1.8366 -0.0903 0.0627>
   rotation=<rot 0 92.9542 -78.2406>
  }]
 dyn_hair_ponytail_14 [{
   shapeType=Capsule
   length=0.505
   radius=0.505
   shapeName=dyn_hair_ponytail_14_capsule
   center=<0.8312 0.0857 0.0415>
   rotation=<rot 0 -95.6363 -73.8547>
  }]
 dyn_hair_ponytail_15 [{
   shapeType=Capsule
   length=0.505
   radius=0.505
   shapeName=dyn_hair_ponytail_15_capsule
   center=<0.8241 -0.0229 0.0085>
   rotation=<rot 0 65.1055 -76.9167>
  }]
 head [{shapeType=Capsule, length=7.6604, radius=7.6766, center=<1.6272 0.4003 -0.3842>, rotation=<rot 0 -90 -90>}]
}
constraints [
 {bone1Name=dyn_hair_front_01_l_root, bone2Name=head, swing1LimitDegrees=10, swing2LimitDegrees=10, twistLimitDegrees=10}
 {bone1Name=dyn_hair_front_01_l, bone2Name=dyn_hair_front_01_l_root, swing1LimitDegrees=10, swing2LimitDegrees=10, twistLimitDegrees=10}
 {bone1Name=dyn_hair_front_02_l, bone2Name=dyn_hair_front_01_l, swing1LimitDegrees=10, swing2LimitDegrees=10, twistLimitDegrees=10}
 {bone1Name=dyn_hair_front_03_l, bone2Name=dyn_hair_front_02_l, swing1LimitDegrees=10, swing2LimitDegrees=10, twistLimitDegrees=10}
 {bone1Name=dyn_hair_front_04_l, bone2Name=dyn_hair_front_03_l, swing1LimitDegrees=10, swing2LimitDegrees=10, twistLimitDegrees=10}
 {bone1Name=dyn_hair_front_01_r_root, bone2Name=head, swing1LimitDegrees=10, swing2LimitDegrees=10, twistLimitDegrees=10}
 {bone1Name=dyn_hair_front_01_r, bone2Name=dyn_hair_front_01_r_root, swing1LimitDegrees=14, swing2LimitDegrees=14, twistLimitDegrees=14}
 {bone1Name=dyn_hair_front_02_r, bone2Name=dyn_hair_front_01_r, swing1LimitDegrees=14, swing2LimitDegrees=14, twistLimitDegrees=14}
 {bone1Name=dyn_hair_front_03_r, bone2Name=dyn_hair_front_02_r, swing1LimitDegrees=14, swing2LimitDegrees=14, twistLimitDegrees=14}
 {bone1Name=dyn_hair_front_04_r, bone2Name=dyn_hair_front_03_r, swing1LimitDegrees=14, swing2LimitDegrees=14, twistLimitDegrees=14}
 {bone1Name=dyn_hair_ponytail_04, bone2Name=head, swing1LimitDegrees=24, swing2LimitDegrees=24, twistLimitDegrees=6}
 {bone1Name=dyn_hair_ponytail_05, bone2Name=dyn_hair_ponytail_04, swing1LimitDegrees=24, swing2LimitDegrees=24, twistLimitDegrees=6}
 {bone1Name=dyn_hair_ponytail_12, bone2Name=dyn_hair_ponytail_04, swing1LimitDegrees=16, swing2LimitDegrees=14, twistLimitDegrees=4}
 {bone1Name=dyn_hair_ponytail_13, bone2Name=dyn_hair_ponytail_12, swing1LimitDegrees=16, swing2LimitDegrees=14, twistLimitDegrees=4}
 {bone1Name=dyn_hair_ponytail_14, bone2Name=dyn_hair_ponytail_04, swing1LimitDegrees=16, swing2LimitDegrees=14, twistLimitDegrees=4}
 {bone1Name=dyn_hair_ponytail_15, bone2Name=dyn_hair_ponytail_14, swing1LimitDegrees=16, swing2LimitDegrees=14, twistLimitDegrees=4}
 {bone1Name=dyn_hair_front_07_r, bone2Name=dyn_hair_front_02_r, swing1LimitDegrees=10, swing2LimitDegrees=10, twistLimitDegrees=10}
 {bone1Name=dyn_hair_front_08_r, bone2Name=dyn_hair_front_07_r, swing1LimitDegrees=10, swing2LimitDegrees=10, twistLimitDegrees=10}
 {bone1Name=dyn_hair_front_05_r, bone2Name=dyn_hair_front_01_r, swing1LimitDegrees=10, swing2LimitDegrees=10, twistLimitDegrees=10}
 {bone1Name=dyn_hair_front_06_r, bone2Name=dyn_hair_front_05_r, swing1LimitDegrees=10, swing2LimitDegrees=10, twistLimitDegrees=10}
]
