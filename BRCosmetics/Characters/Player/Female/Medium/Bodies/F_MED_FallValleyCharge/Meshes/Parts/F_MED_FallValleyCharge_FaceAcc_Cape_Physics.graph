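# Physics asset for the FaceAcc_Cape part of the female medium FallValleyCharge player cosmetic, giving its dyn_hair_ponytail bones capsule bodies for secondary motion.
# The chain runs from a root under the head through segment 06, then branches to 07, which forks into 08 and 09, and to 10–11.
# Tight swing and twist limits shape the chain.
# Head, upper-spine and clavicle capsules supply body volume, with angularly free joints, as is the root-to-head joint.
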PhysicsAsset /BRCosmetics/Characters/Player/Female/Medium/Bodies/F_MED_FallValleyCharge/Meshes/Parts/F_MED_FallValleyCharge_FaceAcc_Cape_Physics
bodies {
 clavicle_l [{
   shapeType=Capsule
   length=25.375
   radius=10.9738
   shapeName=clavicle_l_capsule
   center=<9.4131 1.5381 1.4487>
   rotation=<rot 0.246 -98.4815 106.2986>
  }]
 clavicle_r [{
   shapeType=Capsule
   length=25.375
   radius=10.9738
   shapeName=clavicle_l_capsule
   center=<-9.4131 -1.5381 -1.4487>
   rotation=<rot 0.5675 -99.5809 -73.7092>
  }]
 dyn_hair_ponytail_01 [{shapeType=Capsule, length=4.0567, radius=3.1843, shapeName=dyn_hair_ponytail_01_capsule, center=<1 0 0>, rotation=<rot 90 0 0>}]
 dyn_hair_ponytail_01_root [{shapeType=Capsule, length=0.505, radius=0.505, shapeName=dyn_hair_ponytail_01_root_capsule, rotation=<rot 0 -90 0>}]
 dyn_hair_ponytail_02 [{shapeType=Capsule, length=4.4633, radius=2.8871, shapeName=dyn_hair_ponytail_02_capsule, center=<1 0 0>, rotation=<rot 90 0 0>}]
 dyn_hair_ponytail_03 [{shapeType=Capsule, length=2.3469, radius=3.9234, shapeName=dyn_hair_ponytail_03_capsule, center=<1 0 0>, rotation=<rot 90 0 0>}]
 dyn_hair_ponytail_04 [{shapeType=Capsule, length=2.0559, radius=3.6152, shapeName=dyn_hair_ponytail_04_capsule, center=<1 0 0>, rotation=<rot 90 0 0>}]
 dyn_hair_ponytail_05 [{shapeType=Capsule, length=2.1156, radius=3.4873, shapeName=dyn_hair_ponytail_05_capsule, center=<1 0 0>, rotation=<rot 90 0 0>}]
 dyn_hair_ponytail_06 [{
   shapeType=Capsule
   length=3.9901
   radius=3.5723
   shapeName=dyn_hair_ponytail_06_capsule
   center=<5.3734 0.4022 0>
   rotation=<rot 90 0 -9.7904>
  }]
 dyn_hair_ponytail_07 [{
   shapeType=Capsule
   length=3.7726
   radius=1.3777
   shapeName=dyn_hair_ponytail_07_capsule
   center=<1.6042 0 0>
   rotation=<rot -12.8296 -91.9434 -5.8258>
  }]
 dyn_hair_ponytail_08 [{
   shapeType=Capsule
   length=1.8689
   radius=0.934
   shapeName=dyn_hair_ponytail_08_capsule
   center=<0.9449 0 0>
   rotation=<rot 0 88.3357 -80.1443>
  }]
 dyn_hair_ponytail_09 [{
   shapeType=Capsule
   length=0.505
   radius=0.505
   shapeName=dyn_hair_ponytail_09_capsule
   center=<1.1212 0 0>
   rotation=<rot 0 100.0666 -74.7959>
  }]
 dyn_hair_ponytail_10 [{shapeType=Capsule, length=3.113, radius=1.1834, shapeName=dyn_hair_ponytail_10_capsule, center=<1.6215 0 0>, rotation=<rot 0 -90 0>}]
 dyn_hair_ponytail_11 [{
   shapeType=Capsule
   length=0.5312
   radius=1.38
   shapeName=dyn_hair_ponytail_11_capsule
   center=<1.7543 0 0>
   rotation=<rot 0 -112.9384 -178.4279>
  }]
 head [{shapeType=Capsule, length=7.6604, radius=7.6766, center=<1.6272 0.4003 -0.3842>, rotation=<rot 0 -90 -90>}]
 spine_05 [
  {shapeType=Capsule, length=16.5938, radius=10.4425, shapeName=spine_05_capsule, center=<2.898 1.7809 0>, rotation=<rot 0 -76.9053 90>}
  {shapeType=Capsule, length=15, radius=15, shapeName=spine_05_capsule, center=<3.0187 -2.546 0>}
 ]
}
constraints [
 {bone1Name=dyn_hair_ponytail_01, bone2Name=dyn_hair_ponytail_01_root, swing1LimitDegrees=6, swing2LimitDegrees=5, twistLimitDegrees=6}
 {bone1Name=dyn_hair_ponytail_02, bone2Name=dyn_hair_ponytail_01, swing1LimitDegrees=24, swing2LimitDegrees=5, twistLimitDegrees=6}
 {bone1Name=dyn_hair_ponytail_03, bone2Name=dyn_hair_ponytail_02, swing1LimitDegrees=30, swing2LimitDegrees=5, twistLimitDegrees=6}
 {bone1Name=dyn_hair_ponytail_04, bone2Name=dyn_hair_ponytail_03, swing1LimitDegrees=30, swing2LimitDegrees=5, twistLimitDegrees=6}
 {bone1Name=dyn_hair_ponytail_05, bone2Name=dyn_hair_ponytail_04, swing1LimitDegrees=40, swing2LimitDegrees=5, twistLimitDegrees=6}
 {bone1Name=dyn_hair_ponytail_06, bone2Name=dyn_hair_ponytail_05, swing1LimitDegrees=40, swing2LimitDegrees=5, twistLimitDegrees=6}
 {
  bone1Name=head
  bone2Name=spine_05
  swing1LimitDegrees=45
  swing1Motion=Free
  swing2LimitDegrees=45
  swing2Motion=Free
  twistLimitDegrees=45
  twistMotion=Free
 }
 {
  bone1Name=clavicle_l
  bone2Name=spine_05
  swing1LimitDegrees=45
  swing1Motion=Free
  swing2LimitDegrees=45
  swing2Motion=Free
  twistLimitDegrees=45
  twistMotion=Free
 }
 {bone1Name=dyn_hair_ponytail_07, bone2Name=dyn_hair_ponytail_06, swing1LimitDegrees=24, swing2LimitDegrees=5, twistLimitDegrees=6}
 {bone1Name=dyn_hair_ponytail_08, bone2Name=dyn_hair_ponytail_07, swing1LimitDegrees=24, swing2LimitDegrees=5, twistLimitDegrees=6}
 {bone1Name=dyn_hair_ponytail_09, bone2Name=dyn_hair_ponytail_07, swing1LimitDegrees=24, swing2LimitDegrees=5, twistLimitDegrees=6}
 {bone1Name=dyn_hair_ponytail_10, bone2Name=dyn_hair_ponytail_06, swing1LimitDegrees=24, swing2LimitDegrees=14, twistLimitDegrees=6}
 {bone1Name=dyn_hair_ponytail_11, bone2Name=dyn_hair_ponytail_10, swing1LimitDegrees=24, swing2LimitDegrees=14, twistLimitDegrees=6}
 {
  bone1Name=clavicle_r
  bone2Name=spine_05
  swing1LimitDegrees=45
  swing1Motion=Free
  swing2LimitDegrees=45
  swing2Motion=Free
  twistLimitDegrees=45
  twistMotion=Free
 }
 {
  bone1Name=dyn_hair_ponytail_01_root
  bone2Name=head
  swing1LimitDegrees=45
  swing1Motion=Free
  swing2LimitDegrees=45
  swing2Motion=Free
  twistLimitDegrees=45
  twistMotion=Free
 }
]
